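# Sound cue for the CyberArmorFemale cosmetic pickaxe, apparently played when it hits an enemy.
# A fixed short delay leads into an equally weighted random pick among six ImpactB wave variations.
# It uses shared pickaxe impact attenuation and property sheet, the Fort_Impacts sound class, and the SwingFX_Preset2 source effect chain.
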SoundCue /BRCosmetics/Sounds/PickAxes/CyberArmorFemale/PickaxeImpactEnemyE_CyberArmorFemale
root=Delay_0
nodes [
 {id=Delay_0, children [Random_0], properties {delayMax=0.1, delayMin=0.1}}
 {id=Random_0, children [WavePlayer_0,WavePlayer_1,WavePlayer_2,WavePlayer_3,WavePlayer_4,WavePlayer_5], properties {weights [1,1,1,1,1,1]}}
 {id=WavePlayer_0, properties {soundWaveAssetPtr=&/CyberArmorFemale_ImpactB_01}}
 {id=WavePlayer_1, properties {soundWaveAssetPtr=&/CyberArmorFemale_ImpactB_02}}
 {id=WavePlayer_2, properties {soundWaveAssetPtr=&/CyberArmorFemale_ImpactB_03}}
 {id=WavePlayer_3, properties {soundWaveAssetPtr=&/CyberArmorFemale_ImpactB_04}}
 {id=WavePlayer_4, properties {soundWaveAssetPtr=&/CyberArmorFemale_ImpactB_05}}
 {id=WavePlayer_5, properties {soundWaveAssetPtr=&/CyberArmorFemale_ImpactB_06}}
]
properties {
 attenuationSettings=/Game/Sounds/Attenuation/Fort_Pickaxe_Impact_Att
 audioPropertiesSheet=/BRCosmetics/Sounds/Settings/PropertySheets/Pickaxe/APS_Cosmetic_Pickaxe_Impact
 maxDistance=4300
 soundClassObject=&Fort_Impacts
 sourceEffectChain=&/SwingFX_Preset2
 concurrencySet [null]
}
