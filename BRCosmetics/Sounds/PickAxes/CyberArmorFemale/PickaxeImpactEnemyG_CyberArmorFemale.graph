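# Sound cue for the CyberArmorFemale cosmetic pickaxe, played on impact with an enemy, per its name.
# A fixed short delay leads into an equally weighted random pick among six ImpactA wave variations.
# It uses the shared pickaxe impact attenuation, property sheet and impacts sound class, plus a SwingFX preset source effect chain.
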SoundCue /BRCosmetics/Sounds/PickAxes/CyberArmorFemale/PickaxeImpactEnemyG_CyberArmorFemale
root=Delay_0
nodes [
 {id=Delay_0, children [Random_0], properties {delayMax=0.1, delayMin=0.1}}
 {id=Random_0, children [WavePlayer_0,WavePlayer_1,WavePlayer_2,WavePlayer_3,WavePlayer_4,WavePlayer_5], properties {weights [1,1,1,1,1,1]}}
 {id=WavePlayer_0, properties {soundWaveAssetPtr=&/CyberArmorFemale_ImpactA_01}}
 {id=WavePlayer_1, properties {soundWaveAssetPtr=&/CyberArmorFemale_ImpactA_02}}
 {id=WavePlayer_2, properties {soundWaveAssetPtr=&/CyberArmorFemale_ImpactA_03}}
 {id=WavePlayer_3, properties {soundWaveAssetPtr=&/CyberArmorFemale_ImpactA_04}}
 {id=WavePlayer_4, properties {soundWaveAssetPtr=&/CyberArmorFemale_ImpactA_05}}
 {id=WavePlayer_5, properties {soundWaveAssetPtr=&/CyberArmorFemale_ImpactA_06}}
]
properties {
 attenuationSettings=/Game/Sounds/Attenuation/Fort_Pickaxe_Impact_Att
 audioPropertiesSheet=/BRCosmetics/Sounds/Settings/PropertySheets/Pickaxe/APS_Cosmetic_Pickaxe_Impact
 maxDistance=4300
 soundClassObject=&Fort_Impacts
 sourceEffectChain=&/SwingFX_Preset6
 concurrencySet [null]
}
